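# Physics asset for the GuardHence back bling (backpack cosmetic) mesh, giving its two hanging cables a ragdoll-style physics setup.
# Capsule bodies cover the pack and two chains, each a base cable plus three dyn segments.
# Both base cables are angularly locked to the pack, while each dyn segment links to its parent with limited swing and twist.
# The left chain allows wider swing on one axis, and the right uses thicker segment capsules.
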PhysicsAsset /BRCosmetics/Accessories/FORT_Backpacks/Backpack_M_MED_GuardHence/Meshes/M_MED_GuardHence_Pack_Physics
bodies {
 base_cable_l [{shapeType=Capsule, length=5, radius=1.4, shapeName=base_cable_l_capsule, center=<1 0 0>}]
 base_cable_r [{shapeType=Capsule, length=5, radius=1.4, shapeName=base_cable_r_capsule, center=<-0.3322 0 0>}]
 dyn_cable_l_01 [{shapeType=Capsule, length=8, radius=1.4, shapeName=dyn_cable_l_01_capsule, center=<0.6551 0 0>}]
 dyn_cable_l_02 [{shapeType=Capsule, length=8, radius=1.4, shapeName=dyn_cable_l_02_capsule, center=<1.6085 0 0>}]
 dyn_cable_l_03 [{shapeType=Capsule, length=8, radius=1.4, shapeName=dyn_cable_l_03_capsule, center=<1.7015 0 0>}]
 dyn_cable_r_01 [{shapeType=Capsule, length=8, radius=2, shapeName=dyn_cable_r_01_capsule, center=<0.0449 0 0>}]
 dyn_cable_r_02 [{shapeType=Capsule, length=8, radius=2, shapeName=dyn_cable_r_02_capsule, center=<2.0629 0 0>}]
 dyn_cable_r_03 [{shapeType=Capsule, length=8, radius=2, shapeName=dyn_cable_r_03_capsule, center=<3.1168 0 0>}]
 pack [{shapeType=Capsule, length=10, radius=10, shapeName=pack_capsule, center=<0 8.7872 -16.7619>, rotation=<rot 90 0 0>}]
}
constraints [
 {bone1Name=dyn_cable_l_01, bone2Name=base_cable_l, swing1LimitDegrees=25, swing2LimitDegrees=10, twistLimitDegrees=10}
 {bone1Name=dyn_cable_l_02, bone2Name=dyn_cable_l_01, swing1LimitDegrees=25, swing2LimitDegrees=10, twistLimitDegrees=10}
 {bone1Name=dyn_cable_l_03, bone2Name=dyn_cable_l_02, swing1LimitDegrees=25, swing2LimitDegrees=10, twistLimitDegrees=10}
 {bone1Name=dyn_cable_r_01, bone2Name=base_cable_r, swing1LimitDegrees=10, swing2LimitDegrees=10, twistLimitDegrees=10}
 {bone1Name=dyn_cable_r_02, bone2Name=dyn_cable_r_01, swing1LimitDegrees=10, swing2LimitDegrees=10, twistLimitDegrees=10}
 {bone1Name=dyn_cable_r_03, bone2Name=dyn_cable_r_02, swing1LimitDegrees=10, swing2LimitDegrees=10, twistLimitDegrees=10}
 {
  bone1Name=base_cable_r
  bone2Name=pack
  swing1LimitDegrees=10
  swing1Motion=Locked
  swing2LimitDegrees=10
  swing2Motion=Locked
  twistLimitDegrees=10
  twistMotion=Locked
 }
 {
  bone1Name=base_cable_l
  bone2Name=pack
  swing1LimitDegrees=10
  swing1Motion=Locked
  swing2LimitDegrees=10
  swing2Motion=Locked
  twistLimitDegrees=10
  twistMotion=Locked
 }
]
