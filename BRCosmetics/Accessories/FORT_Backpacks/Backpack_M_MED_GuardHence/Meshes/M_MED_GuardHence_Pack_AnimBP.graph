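# Animation Blueprint for the GuardHence backpack cosmetic that adds physics-driven secondary motion to the pack.
# The anim graph runs a RigidBody node on the input pose, simulating the pack's physics asset in root-bone space with clamped world motion and linear drag.
# Each update, a nativized gravity override state function computes a vector from a parameter asset, stores it in Gravity, and feeds it to the node as world gravity.
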AnimBlueprint /BRCosmetics/Accessories/FORT_Backpacks/Backpack_M_MED_GuardHence/Meshes/M_MED_GuardHence_Pack_AnimBP
target_skeleton=&/M_MED_GuardHence_Pack_Skeleton
graphs [
 AnimGraph {nodes [
   n1 = RigidBody!(
    ComponentPose=Local To Component(LocalPose=Input Pose())
    Alpha
    OverrideWorldGravity:Vector=Get Gravity() -> (Gravity:Vector)
    ExternalForce
   ) -> (Pose) {
    bClampLinearTranslationLimitToRefPose=true
    bOverrideWorldGravity=true
    overridePhysicsAsset=&/M_MED_GuardHence_Pack_Physics
    simulationSpace=BaseBoneSpace
    baseBoneRef {boneName=root}
    simSpaceSettings {
     dampingAlpha=0
     maxAngularAcceleration=21
     maxAngularVelocity=8
     maxLinearAcceleration=1500
     maxLinearVelocity=700
     worldAlpha=1
     externalLinearDragV=<0.75 0.75 0.75>
    }
   }
   Output Pose(Result=Component To Local(ComponentPose=@n1)) {layerGroup=None}
  ]}
 EventGraph {nodes [
   Event BlueprintUpdateAnimation()
   > n1 = "Gravity Override State Function (Parameter Asset) (Nativized)"(
    CharacterPartAnimInstance=Self-Reference()
    GameTimeInSeconds=GetGameTimeInSeconds()
    Parameters=&/M_MED_GuardHense_Pack_Parameters
   )
   > Set Gravity(Gravity:Vector=@n1) -> (Output_Get:Vector)
   TryGetPawnOwner()
  ]}
]
variables [Gravity {type=Vector}]
